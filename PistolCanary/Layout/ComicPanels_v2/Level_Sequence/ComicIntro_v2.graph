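# ComicIntro_v2 is the PistolCanary comic intro Level Sequence, playing three back-to-back cinematic shots at 60 fps.
# An event trigger fires on the comic panel manager, and a visibility track hides the NoFightBarrier actor.
# An MPC_ComicIntro track animates the Color parameter's RGB from white to black over the opening frames, alongside a console variable track.
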LevelSequence /PistolCanary/Layout/ComicPanels_v2/Level_Sequence/ComicIntro_v2
bindings [
 BP_PistolCanary_NoFightBarrier {
  id=#1
  possessed_class=BP_PistolCanary_NoFightBarrier_C
  tracks [{
    type=MovieSceneVisibilityTrack
    sections [{type=MovieSceneVisibilitySection, channels [None {type=MovieSceneScriptingBoolChannel, default=true, const [false,1]}]}]
   }]
 }
 PistolCanary_PanelManager {
  id=#2
  possessed_class=PistolCanary_PanelManager_C
  tracks [{
    type=MovieSceneEventTrack
    sections [{
      type=MovieSceneEventTriggerSection
      channels [None {type=MovieSceneScriptingEventChannel, const ["<Struct 'MovieSceneEvent' (0x85285c43200) {}>", 1]}]
     }]
   }]
 }
]
display_rate {denominator=1, numerator=60}
tick_resolution {denominator=1, numerator=24000}
tracks [
 {
  type=MovieSceneCinematicShotTrack
  display_name=Shots
  sections [
   {type=MovieSceneCinematicShotSection, end_frame=434, start_frame=0}
   {type=MovieSceneCinematicShotSection, end_frame=634, start_frame=434}
   {type=MovieSceneCinematicShotSection, end_frame=813, start_frame=634}
  ]
 }
 {type=MovieSceneCVarTrack, display_name=Console Variable, sections [{type=MovieSceneCVarSection}]}
 {
  type=MovieSceneMaterialParameterCollectionTrack
  display_name=MPC_ComicIntro
  sections [{
    type=MovieSceneParameterSection
    channels [
     ColorR {type=MovieSceneScriptingFloatChannel, default=0, keys [[0,1],[48,0]]}
     ColorG {type=MovieSceneScriptingFloatChannel, default=0, keys [[0,1],[48,0]]}
     ColorB {type=MovieSceneScriptingFloatChannel, default=0, keys [[0,1],[48,0]]}
     ColorA {type=MovieSceneScriptingFloatChannel, default=0, const [0,2]}
    ]
   }]
 }
]
