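# Level Sequence for intro comic panel 1, staging a one-camera animated reveal of six characters: Killswitch, Lightrider, Myst, then Synth, Ziggy and Haylee.
# Speed panels slide in ahead of each trio, and a sharp camera pull-back along X separates the two waves.
# Character roots snap into place then drift, while separate body-part and effect actors slowly rotate or scale like layered cutouts.
# Staggered camera shakes accompany the character entrances.
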LevelSequence /PistolCanary/Layout/ComicPanels_v2/Level_Sequence/IntroPanel1
bindings [
 P1_Haylee_L_Hand {
  id=#1
  possessed_class=Actor
  tracks [{
    type=MovieScene3DTransformTrack
    sections [{
      type=MovieScene3DTransformSection
      channels [
       Location.X {type=MovieSceneScriptingDoubleChannel, default=0}
       Location.Y {type=MovieSceneScriptingDoubleChannel, default=22.4746}
       Location.Z {type=MovieSceneScriptingDoubleChannel, default=-20.9196}
       Rotation.X {type=MovieSceneScriptingDoubleChannel, default=0, keys [[315,0],[434,-7.5153]]}
       Rotation.Y {type=MovieSceneScriptingDoubleChannel, default=0, const [0,2]}
       Rotation.Z {type=MovieSceneScriptingDoubleChannel, default=0, const [0,2]}
       Scale.X {type=MovieSceneScriptingDoubleChannel, default=1}
       Scale.Y {type=MovieSceneScriptingDoubleChannel, default=1}
       Scale.Z {type=MovieSceneScriptingDoubleChannel, default=1}
      ]
     }]
   }]
 }
 P1_Kill_Mask {
  id=#2
  possessed_class=Actor
  tracks [{
    type=MovieScene3DTransformTrack
    sections [{
      type=MovieScene3DTransformSection
      channels [
       Location.X {type=MovieSceneScriptingDoubleChannel, default=-0.2}
       Location.Y {type=MovieSceneScriptingDoubleChannel, default=3.9989}
       Location.Z {type=MovieSceneScriptingDoubleChannel, default=19.2495}
       Rotation.X {type=MovieSceneScriptingDoubleChannel, default=0}
       Rotation.Y {type=MovieSceneScriptingDoubleChannel, default=0}
       Rotation.Z {type=MovieSceneScriptingDoubleChannel, default=0}
       Scale.X {type=MovieSceneScriptingDoubleChannel, default=1, keys [[155,0],[165,1],[168,0.9842],[171,1]]}
       Scale.Y {type=MovieSceneScriptingDoubleChannel, default=1, keys [[155,0],[165,1],[168,0.9842],[171,1]]}
       Scale.Z {type=MovieSceneScriptingDoubleChannel, default=1, keys [[155,0],[165,1],[168,0.9842],[171,1]]}
      ]
     }]
   }]
 }
 CameraComponent {
  id=#3
  parent_id=#4
  possessed_class=CineCameraComponent
  tracks [
   {
    type=MovieSceneFloatTrack
    display_name=Current Focal Length
    sections [{type=MovieSceneFloatSection, channels [None {type=MovieSceneScriptingFloatChannel, default=11.8}]}]
   }
   {
    type=MovieSceneFloatTrack
    display_name="Manual Focus Distance (Focus Settings)"
    sections [{type=MovieSceneFloatSection, channels [None {type=MovieSceneScriptingFloatChannel, default=100000}]}]
   }
   {
    type=MovieSceneFloatTrack
    display_name=Current Aperture
    sections [{type=MovieSceneFloatSection, channels [None {type=MovieSceneScriptingFloatChannel, default=2.8}]}]
   }
  ]
 }
 Speed_Panel_5 {
  id=#5
  possessed_class=FortStaticMeshActor
  tracks [{
    type=MovieScene3DTransformTrack
    sections [{
      type=MovieScene3DTransformSection
      channels [
       Location.X {type=MovieSceneScriptingDoubleChannel, default=-64.7571, const [-64.7571,2]}
       Location.Y {type=MovieSceneScriptingDoubleChannel, default=-5.8137, keys [[209,-159.9022],[224,-5.8137]]}
       Location.Z {type=MovieSceneScriptingDoubleChannel, default=3.5504, keys [[209,93.4931],[224,3.5504]]}
       Rotation.X {type=MovieSceneScriptingDoubleChannel, default=90}
       Rotation.Y {type=MovieSceneScriptingDoubleChannel, default=59.7275}
       Rotation.Z {type=MovieSceneScriptingDoubleChannel, default=90}
       Scale.X {type=MovieSceneScriptingDoubleChannel, default=0.425}
       Scale.Y {type=MovieSceneScriptingDoubleChannel, default=2.3563}
       Scale.Z {type=MovieSceneScriptingDoubleChannel, default=1}
      ]
     }]
   }]
 }
 P1_Myst_R_Hand {
  id=#6
  possessed_class=Actor
  tracks [{
    type=MovieScene3DTransformTrack
    sections [{
      type=MovieScene3DTransformSection
      channels [
       Location.X {type=MovieSceneScriptingDoubleChannel, default=0}
       Location.Y {type=MovieSceneScriptingDoubleChannel, default=11.3839}
       Location.Z {type=MovieSceneScriptingDoubleChannel, default=13.9749}
       Rotation.X {type=MovieSceneScriptingDoubleChannel, default=0, keys [[114,0],[240,15.7329]]}
       Rotation.Y {type=MovieSceneScriptingDoubleChannel, default=0, const [0,1]}
       Rotation.Z {type=MovieSceneScriptingDoubleChannel, default=0, const [0,1]}
       Scale.X {type=MovieSceneScriptingDoubleChannel, default=1}
       Scale.Y {type=MovieSceneScriptingDoubleChannel, default=1}
       Scale.Z {type=MovieSceneScriptingDoubleChannel, default=1}
      ]
     }]
   }]
 }
 Speed_Panel_3 {
  id=#7
  possessed_class=FortStaticMeshActor
  tracks [{
    type=MovieScene3DTransformTrack
    sections [{
      type=MovieScene3DTransformSection
      channels [
       Location.X {type=MovieSceneScriptingDoubleChannel, default=48.4861, const [48.4861,2]}
       Location.Y {type=MovieSceneScriptingDoubleChannel, default=28.3664, keys [[0,-215.5819],[15,41.9342]]}
       Location.Z {type=MovieSceneScriptingDoubleChannel, default=-77.3209, keys [[0,-252.1481],[15,-99.4224]]}
       Rotation.X {type=MovieSceneScriptingDoubleChannel, default=-90}
       Rotation.Y {type=MovieSceneScriptingDoubleChannel, default=58.4549}
       Rotation.Z {type=MovieSceneScriptingDoubleChannel, default=-90}
       Scale.X {type=MovieSceneScriptingDoubleChannel, default=1.5}
       Scale.Y {type=MovieSceneScriptingDoubleChannel, default=4.5}
       Scale.Z {type=MovieSceneScriptingDoubleChannel, default=1}
      ]
     }]
   }]
 }
 Panel1_Synth_Root {
  id=#8
  possessed_class=Actor
  tracks [{
    type=MovieScene3DTransformTrack
    sections [{
      type=MovieScene3DTransformSection
      channels [
       Location.X {type=MovieSceneScriptingDoubleChannel, default=-72.027, const [-72.027,2]}
       Location.Y {type=MovieSceneScriptingDoubleChannel, default=-25.8621, keys [[235,-17.3642],[250,-27.0912],[434,-26.2004]]}
       Location.Z {type=MovieSceneScriptingDoubleChannel, default=-17.225, keys [[235,-66.8705],[250,-18.9793],[434,-15.4841]]}
       Rotation.X {type=MovieSceneScriptingDoubleChannel, default=42.6061, keys [[235,21.0006],[250,42.6061],[434,47.019]]}
       Rotation.Y {type=MovieSceneScriptingDoubleChannel, default=0, const [0,2]}
       Rotation.Z {type=MovieSceneScriptingDoubleChannel, default=0, const [0,2]}
       Scale.X {type=MovieSceneScriptingDoubleChannel, default=0.764, keys [[235,0.5885],[250,0.764]]}
       Scale.Y {type=MovieSceneScriptingDoubleChannel, default=0.764, keys [[235,0.5885],[250,0.764]]}
       Scale.Z {type=MovieSceneScriptingDoubleChannel, default=0.764, keys [[235,0.5885],[250,0.764]]}
      ]
     }]
   }]
 }
 Speed_Panel_4 {
  id=#9
  possessed_class=FortStaticMeshActor
  tracks [{
    type=MovieScene3DTransformTrack
    sections [{
      type=MovieScene3DTransformSection
      channels [
       Location.X {type=MovieSceneScriptingDoubleChannel, default=-65.7571, const [-65.7571,2]}
       Location.Y {type=MovieSceneScriptingDoubleChannel, default=21.8398, keys [[215,193.3787],[230,21.8398]]}
       Location.Z {type=MovieSceneScriptingDoubleChannel, default=50.7564, keys [[215,-49.3722],[230,50.7564]]}
       Rotation.X {type=MovieSceneScriptingDoubleChannel, default=90}
       Rotation.Y {type=MovieSceneScriptingDoubleChannel, default=59.7275}
       Rotation.Z {type=MovieSceneScriptingDoubleChannel, default=90}
       Scale.X {type=MovieSceneScriptingDoubleChannel, default=0.7}
       Scale.Y {type=MovieSceneScriptingDoubleChannel, default=3.0294}
       Scale.Z {type=MovieSceneScriptingDoubleChannel, default=1}
      ]
     }]
   }]
 }
 Camera_01 {
  id=#4
  possessed_class=CineCameraActor
  tracks [
   {
    type=MovieScene3DTransformTrack
    sections [{
      type=MovieScene3DTransformSection
      channels [
       Location.X {type=MovieSceneScriptingDoubleChannel, default=0, keys [[0,-42.997],[25,-60.778],[192,-64.6175],[213,-132.765],[434,-137.7943]]}
       Location.Y {type=MovieSceneScriptingDoubleChannel, default=0}
       Location.Z {type=MovieSceneScriptingDoubleChannel, default=0}
       Rotation.X {type=MovieSceneScriptingDoubleChannel, default=0}
       Rotation.Y {type=MovieSceneScriptingDoubleChannel, default=0}
       Rotation.Z {type=MovieSceneScriptingDoubleChannel, default=0}
       Scale.X {type=MovieSceneScriptingDoubleChannel, default=1}
       Scale.Y {type=MovieSceneScriptingDoubleChannel, default=1}
       Scale.Z {type=MovieSceneScriptingDoubleChannel, default=1}
      ]
     }]
   }
   {
    type=MovieSceneCameraShakeTrack
    sections [
     {type=MovieSceneCameraShakeSection, end_frame=110, start_frame=62}
     {type=MovieSceneCameraShakeSection, end_frame=142, start_frame=95}
     {type=MovieSceneCameraShakeSection, end_frame=175, start_frame=130}
     {type=MovieSceneCameraShakeSection, end_frame=297, start_frame=244}
     {type=MovieSceneCameraShakeSection, end_frame=349, start_frame=296}
     {type=MovieSceneCameraShakeSection, end_frame=627, start_frame=327}
    ]
   }
  ]
 }
 Panel1_Lightrider_Root {
  id=#10
  possessed_class=Actor
  tracks [{
    type=MovieScene3DTransformTrack
    sections [{
      type=MovieScene3DTransformSection
      channels [
       Location.X {type=MovieSceneScriptingDoubleChannel, default=22.7982, const [22.7982,3]}
       Location.Y {type=MovieSceneScriptingDoubleChannel, default=34.1376, keys [[74,34.2992],[84,46.1055],[94,34.1376],[240,31.336]]}
       Location.Z {type=MovieSceneScriptingDoubleChannel, default=-18.7162, keys [[74,-1.0783],[84,-10.4759],[94,-18.7162],[240,-19.9504]]}
       Rotation.X {type=MovieSceneScriptingDoubleChannel, default=0, keys [[74,-103.8649],[94,0],[240,3.5374]]}
       Rotation.Y {type=MovieSceneScriptingDoubleChannel, default=0, const [0,3]}
       Rotation.Z {type=MovieSceneScriptingDoubleChannel, default=0, const [0,3]}
       Scale.X {type=MovieSceneScriptingDoubleChannel, default=1, keys [[74,0],[94,1],[240,1]]}
       Scale.Y {type=MovieSceneScriptingDoubleChannel, default=1, keys [[74,0],[94,1],[240,1]]}
       Scale.Z {type=MovieSceneScriptingDoubleChannel, default=1, keys [[74,0],[94,1],[240,1]]}
      ]
     }]
   }]
 }
 Panel1_Myst_Root {
  id=#11
  possessed_class=Actor
  tracks [{
    type=MovieScene3DTransformTrack
    sections [{
      type=MovieScene3DTransformSection
      channels [
       Location.X {type=MovieSceneScriptingDoubleChannel, default=-1.4694, keys [[114,31.9593],[131,-1.4694],[240,-4.7245]]}
       Location.Y {type=MovieSceneScriptingDoubleChannel, default=6.5358, keys [[114,136.8284],[131,6.5358],[240,4.1008]]}
       Location.Z {type=MovieSceneScriptingDoubleChannel, default=6.6443, keys [[114,101.4758],[131,7.0636],[240,4.3559]]}
       Rotation.X {type=MovieSceneScriptingDoubleChannel, default=0, keys [[114,-6.6733],[131,0],[240,1.2725]]}
       Rotation.Y {type=MovieSceneScriptingDoubleChannel, default=0, const [0,3]}
       Rotation.Z {type=MovieSceneScriptingDoubleChannel, default=0, const [0,3]}
       Scale.X {type=MovieSceneScriptingDoubleChannel, default=1, const [1,2]}
       Scale.Y {type=MovieSceneScriptingDoubleChannel, default=1, const [1,2]}
       Scale.Z {type=MovieSceneScriptingDoubleChannel, default=1, const [1,2]}
      ]
     }]
   }]
 }
 P1_Haylee_L_Forearm {
  id=#12
  possessed_class=Actor
  tracks [{
    type=MovieScene3DTransformTrack
    sections [{
      type=MovieScene3DTransformSection
      channels [
       Location.X {type=MovieSceneScriptingDoubleChannel, default=0}
       Location.Y {type=MovieSceneScriptingDoubleChannel, default=18.0353}
       Location.Z {type=MovieSceneScriptingDoubleChannel, default=13.5222}
       Rotation.X {type=MovieSceneScriptingDoubleChannel, default=0, keys [[315,0],[434,-5.2218]]}
       Rotation.Y {type=MovieSceneScriptingDoubleChannel, default=0, const [0,2]}
       Rotation.Z {type=MovieSceneScriptingDoubleChannel, default=0, const [0,2]}
       Scale.X {type=MovieSceneScriptingDoubleChannel, default=0.4588}
       Scale.Y {type=MovieSceneScriptingDoubleChannel, default=0.4588}
       Scale.Z {type=MovieSceneScriptingDoubleChannel, default=0.4588}
      ]
     }]
   }]
 }
 P1_Myst_R_Leg {
  id=#13
  possessed_class=Actor
  tracks [{
    type=MovieScene3DTransformTrack
    sections [{
      type=MovieScene3DTransformSection
      channels [
       Location.X {type=MovieSceneScriptingDoubleChannel, default=0}
       Location.Y {type=MovieSceneScriptingDoubleChannel, default=-22.021}
       Location.Z {type=MovieSceneScriptingDoubleChannel, default=-10.0831}
       Rotation.X {type=MovieSceneScriptingDoubleChannel, default=0, keys [[114,0],[240,7.3793]]}
       Rotation.Y {type=MovieSceneScriptingDoubleChannel, default=0, const [0,1]}
       Rotation.Z {type=MovieSceneScriptingDoubleChannel, default=0, const [0,1]}
       Scale.X {type=MovieSceneScriptingDoubleChannel, default=1}
       Scale.Y {type=MovieSceneScriptingDoubleChannel, default=1}
       Scale.Z {type=MovieSceneScriptingDoubleChannel, default=1}
      ]
     }]
   }]
 }
 Speed_Panel_2 {
  id=#14
  possessed_class=FortStaticMeshActor
  tracks [{
    type=MovieScene3DTransformTrack
    sections [{
      type=MovieScene3DTransformSection
      channels [
       Location.X {type=MovieSceneScriptingDoubleChannel, default=49.4861, const [49.4861,2]}
       Location.Y {type=MovieSceneScriptingDoubleChannel, default=-13.866, keys [[25,324.6186],[40,-13.866]]}
       Location.Z {type=MovieSceneScriptingDoubleChannel, default=-8.5254, keys [[25,199.2647],[40,-8.5254]]}
       Rotation.X {type=MovieSceneScriptingDoubleChannel, default=-90}
       Rotation.Y {type=MovieSceneScriptingDoubleChannel, default=58.4549}
       Rotation.Z {type=MovieSceneScriptingDoubleChannel, default=-90}
       Scale.X {type=MovieSceneScriptingDoubleChannel, default=0.65}
       Scale.Y {type=MovieSceneScriptingDoubleChannel, default=4.5}
       Scale.Z {type=MovieSceneScriptingDoubleChannel, default=1}
      ]
     }]
   }]
 }
 P1_Light_Slide {
  id=#15
  possessed_class=Actor
  tracks [{
    type=MovieScene3DTransformTrack
    sections [{
      type=MovieScene3DTransformSection
      channels [
       Location.X {type=MovieSceneScriptingDoubleChannel, default=22.7982}
       Location.Y {type=MovieSceneScriptingDoubleChannel, default=34.1376}
       Location.Z {type=MovieSceneScriptingDoubleChannel, default=-1.8085}
       Rotation.X {type=MovieSceneScriptingDoubleChannel, default=0, keys [[74,-69.0835],[94,0],[240,1.6585]]}
       Rotation.Y {type=MovieSceneScriptingDoubleChannel, default=0, const [0,3]}
       Rotation.Z {type=MovieSceneScriptingDoubleChannel, default=0, const [0,3]}
       Scale.X {type=MovieSceneScriptingDoubleChannel, default=1, keys [[74,0],[94,1],[240,1.0175]]}
       Scale.Y {type=MovieSceneScriptingDoubleChannel, default=1, keys [[74,0],[94,1],[240,1.0175]]}
       Scale.Z {type=MovieSceneScriptingDoubleChannel, default=1, keys [[74,0],[94,1],[240,1.0175]]}
      ]
     }]
   }]
 }
 ComicIntro_Root {id=#16, possessed_class=TargetPoint, tracks []}
 Speed_Panel_6 {
  id=#17
  possessed_class=FortStaticMeshActor
  tracks [{
    type=MovieScene3DTransformTrack
    sections [{
      type=MovieScene3DTransformSection
      channels [
       Location.X {type=MovieSceneScriptingDoubleChannel, default=-65.7571, const [-65.7571,2]}
       Location.Y {type=MovieSceneScriptingDoubleChannel, default=-33.4052, keys [[204,138.8961],[219,-33.4052]]}
       Location.Z {type=MovieSceneScriptingDoubleChannel, default=-44.0915, keys [[204,-144.6652],[219,-44.0915]]}
       Rotation.X {type=MovieSceneScriptingDoubleChannel, default=90}
       Rotation.Y {type=MovieSceneScriptingDoubleChannel, default=59.7275}
       Rotation.Z {type=MovieSceneScriptingDoubleChannel, default=90}
       Scale.X {type=MovieSceneScriptingDoubleChannel, default=0.7}
       Scale.Y {type=MovieSceneScriptingDoubleChannel, default=2}
       Scale.Z {type=MovieSceneScriptingDoubleChannel, default=1}
      ]
     }]
   }]
 }
 P1_Ziggy_R_Forearm {
  id=#18
  possessed_class=Actor
  tracks [{
    type=MovieScene3DTransformTrack
    sections [{
      type=MovieScene3DTransformSection
      channels [
       Location.X {type=MovieSceneScriptingDoubleChannel, default=0}
       Location.Y {type=MovieSceneScriptingDoubleChannel, default=-33.5848}
       Location.Z {type=MovieSceneScriptingDoubleChannel, default=-18.4329}
       Rotation.X {type=MovieSceneScriptingDoubleChannel, default=0, keys [[282,0],[434,7.8537]]}
       Rotation.Y {type=MovieSceneScriptingDoubleChannel, default=0, const [0,2]}
       Rotation.Z {type=MovieSceneScriptingDoubleChannel, default=0, const [0,2]}
       Scale.X {type=MovieSceneScriptingDoubleChannel, default=0.6541}
       Scale.Y {type=MovieSceneScriptingDoubleChannel, default=0.6541}
       Scale.Z {type=MovieSceneScriptingDoubleChannel, default=0.6541}
      ]
     }]
   }]
 }
 Panel1_Killswitch_Root {
  id=#19
  possessed_class=Actor
  tracks [{
    type=MovieScene3DTransformTrack
    sections [{
      type=MovieScene3DTransformSection
      channels [
       Location.X {type=MovieSceneScriptingDoubleChannel, default=1, keys [[50,1],[65,6.4762],[240,6.4762]]}
       Location.Y {type=MovieSceneScriptingDoubleChannel, default=-34.1952, keys [[50,-129.7753],[65,-39.6207],[240,-33.29]]}
       Location.Z {type=MovieSceneScriptingDoubleChannel, default=7.3592, keys [[50,-36.6323],[65,7.8661],[240,7.6856]]}
       Rotation.X {type=MovieSceneScriptingDoubleChannel, default=0, keys [[50,-14.2098],[65,0],[240,2.156]]}
       Rotation.Y {type=MovieSceneScriptingDoubleChannel, default=0, const [0,3]}
       Rotation.Z {type=MovieSceneScriptingDoubleChannel, default=0, const [0,3]}
       Scale.X {type=MovieSceneScriptingDoubleChannel, default=1}
       Scale.Y {type=MovieSceneScriptingDoubleChannel, default=1}
       Scale.Z {type=MovieSceneScriptingDoubleChannel, default=1}
      ]
     }]
   }]
 }
 P1_Synth_L_Power {
  id=#20
  possessed_class=Actor
  tracks [{
    type=MovieScene3DTransformTrack
    sections [{
      type=MovieScene3DTransformSection
      channels [
       Location.X {type=MovieSceneScriptingDoubleChannel, default=0}
       Location.Y {type=MovieSceneScriptingDoubleChannel, default=-31.0865}
       Location.Z {type=MovieSceneScriptingDoubleChannel, default=7.0186}
       Rotation.X {type=MovieSceneScriptingDoubleChannel, default=0, keys [[235,0],[434,37.3227]]}
       Rotation.Y {type=MovieSceneScriptingDoubleChannel, default=0, const [0,1]}
       Rotation.Z {type=MovieSceneScriptingDoubleChannel, default=0, const [0,1]}
       Scale.X {type=MovieSceneScriptingDoubleChannel, default=0.6005}
       Scale.Y {type=MovieSceneScriptingDoubleChannel, default=0.6005}
       Scale.Z {type=MovieSceneScriptingDoubleChannel, default=0.6005}
      ]
     }]
   }]
 }
 P1_Haylee_Lightning {
  id=#21
  possessed_class=Actor
  tracks [{
    type=MovieScene3DTransformTrack
    sections [{
      type=MovieScene3DTransformSection
      channels [
       Location.X {type=MovieSceneScriptingDoubleChannel, default=0}
       Location.Y {type=MovieSceneScriptingDoubleChannel, default=-1.5102}
       Location.Z {type=MovieSceneScriptingDoubleChannel, default=1.623}
       Rotation.X {type=MovieSceneScriptingDoubleChannel, default=0}
       Rotation.Y {type=MovieSceneScriptingDoubleChannel, default=0}
       Rotation.Z {type=MovieSceneScriptingDoubleChannel, default=0}
       Scale.X {type=MovieSceneScriptingDoubleChannel, default=0.4588, keys [[315,0.4588],[434,0.4975]]}
       Scale.Y {type=MovieSceneScriptingDoubleChannel, default=0.4588, keys [[315,0.4588],[434,0.4975]]}
       Scale.Z {type=MovieSceneScriptingDoubleChannel, default=0.4588, keys [[315,0.4588],[434,0.4975]]}
      ]
     }]
   }]
 }
 Panel1_Ziggy_Root {
  id=#22
  possessed_class=Actor
  tracks [{
    type=MovieScene3DTransformTrack
    sections [{
      type=MovieScene3DTransformSection
      channels [
       Location.X {type=MovieSceneScriptingDoubleChannel, default=-72.027, const [-72.027,3]}
       Location.Y {type=MovieSceneScriptingDoubleChannel, default=34.7872, keys [[282,121.7154],[290,71.783],[297,38.2112],[434,35.7398]]}
       Location.Z {type=MovieSceneScriptingDoubleChannel, default=19.8014, keys [[282,25.9298],[290,29.2554],[297,22.7092],[434,21.0814]]}
       Rotation.X {type=MovieSceneScriptingDoubleChannel, default=50.7825, keys [[282,76.8499],[297,54.2298],[434,48.987]]}
       Rotation.Y {type=MovieSceneScriptingDoubleChannel, default=0, const [0,3]}
       Rotation.Z {type=MovieSceneScriptingDoubleChannel, default=0, const [0,3]}
       Scale.X {type=MovieSceneScriptingDoubleChannel, default=0.7014}
       Scale.Y {type=MovieSceneScriptingDoubleChannel, default=0.7014}
       Scale.Z {type=MovieSceneScriptingDoubleChannel, default=0.7014}
      ]
     }]
   }]
 }
 P1_Kill_L_Foot {
  id=#23
  possessed_class=Actor
  tracks [{
    type=MovieScene3DTransformTrack
    sections [{
      type=MovieScene3DTransformSection
      channels [
       Location.X {type=MovieSceneScriptingDoubleChannel, default=0.2}
       Location.Y {type=MovieSceneScriptingDoubleChannel, default=-11.8064}
       Location.Z {type=MovieSceneScriptingDoubleChannel, default=-5.3281}
       Rotation.X {type=MovieSceneScriptingDoubleChannel, default=0, keys [[50,0],[240,12.7298]]}
       Rotation.Y {type=MovieSceneScriptingDoubleChannel, default=0, const [0,2]}
       Rotation.Z {type=MovieSceneScriptingDoubleChannel, default=0, const [0,2]}
       Scale.X {type=MovieSceneScriptingDoubleChannel, default=1}
       Scale.Y {type=MovieSceneScriptingDoubleChannel, default=1}
       Scale.Z {type=MovieSceneScriptingDoubleChannel, default=1}
      ]
     }]
   }]
 }
 P1_Myst_L_Hand {
  id=#24
  possessed_class=Actor
  tracks [{
    type=MovieScene3DTransformTrack
    sections [{
      type=MovieScene3DTransformSection
      channels [
       Location.X {type=MovieSceneScriptingDoubleChannel, default=0}
       Location.Y {type=MovieSceneScriptingDoubleChannel, default=4.2249}
       Location.Z {type=MovieSceneScriptingDoubleChannel, default=-12.3505}
       Rotation.X {type=MovieSceneScriptingDoubleChannel, default=0, keys [[114,0],[240,13.2683]]}
       Rotation.Y {type=MovieSceneScriptingDoubleChannel, default=0, const [0,1]}
       Rotation.Z {type=MovieSceneScriptingDoubleChannel, default=0, const [0,1]}
       Scale.X {type=MovieSceneScriptingDoubleChannel, default=1}
       Scale.Y {type=MovieSceneScriptingDoubleChannel, default=1}
       Scale.Z {type=MovieSceneScriptingDoubleChannel, default=1}
      ]
     }]
   }]
 }
 P1_Synth_R_Power {
  id=#25
  possessed_class=Actor
  tracks [{
    type=MovieScene3DTransformTrack
    sections [{
      type=MovieScene3DTransformSection
      channels [
       Location.X {type=MovieSceneScriptingDoubleChannel, default=-0.2}
       Location.Y {type=MovieSceneScriptingDoubleChannel, default=-19.3575}
       Location.Z {type=MovieSceneScriptingDoubleChannel, default=-23.3354}
       Rotation.X {type=MovieSceneScriptingDoubleChannel, default=-160, const [-100.2603,1]}
       Rotation.Y {type=MovieSceneScriptingDoubleChannel, default=64.6, const [64.6,1]}
       Rotation.Z {type=MovieSceneScriptingDoubleChannel, default=-72.0518, const [-72.0518,1]}
       Scale.X {type=MovieSceneScriptingDoubleChannel, default=0.6005}
       Scale.Y {type=MovieSceneScriptingDoubleChannel, default=0.6005}
       Scale.Z {type=MovieSceneScriptingDoubleChannel, default=0.6005}
      ]
     }]
   }]
 }
 P1_Kill_R_Hand {
  id=#26
  possessed_class=Actor
  tracks [{
    type=MovieScene3DTransformTrack
    sections [{
      type=MovieScene3DTransformSection
      channels [
       Location.X {type=MovieSceneScriptingDoubleChannel, default=0, const [-0.2,1]}
       Location.Y {type=MovieSceneScriptingDoubleChannel, default=16.1232, const [16.1232,1]}
       Location.Z {type=MovieSceneScriptingDoubleChannel, default=12.9894, const [12.9894,1]}
       Rotation.X {type=MovieSceneScriptingDoubleChannel, default=0, keys [[50,0],[240,-7.8279]]}
       Rotation.Y {type=MovieSceneScriptingDoubleChannel, default=0, const [0,2]}
       Rotation.Z {type=MovieSceneScriptingDoubleChannel, default=0, const [0,2]}
       Scale.X {type=MovieSceneScriptingDoubleChannel, default=1, const [1,1]}
       Scale.Y {type=MovieSceneScriptingDoubleChannel, default=1, const [1,1]}
       Scale.Z {type=MovieSceneScriptingDoubleChannel, default=1, const [1,1]}
      ]
     }]
   }]
 }
 P1_Light_R_Hand {
  id=#27
  possessed_class=Actor
  tracks [{
    type=MovieScene3DTransformTrack
    sections [{
      type=MovieScene3DTransformSection
      channels [
       Location.X {type=MovieSceneScriptingDoubleChannel, default=0}
       Location.Y {type=MovieSceneScriptingDoubleChannel, default=10.6451}
       Location.Z {type=MovieSceneScriptingDoubleChannel, default=8.6898}
       Rotation.X {type=MovieSceneScriptingDoubleChannel, default=0, keys [[74,0],[240,12.9249]]}
       Rotation.Y {type=MovieSceneScriptingDoubleChannel, default=0, const [0,2]}
       Rotation.Z {type=MovieSceneScriptingDoubleChannel, default=0, const [0,2]}
       Scale.X {type=MovieSceneScriptingDoubleChannel, default=1}
       Scale.Y {type=MovieSceneScriptingDoubleChannel, default=1}
       Scale.Z {type=MovieSceneScriptingDoubleChannel, default=1}
      ]
     }]
   }]
 }
 P1_Ziggy_Tail {
  id=#28
  possessed_class=Actor
  tracks [{
    type=MovieScene3DTransformTrack
    sections [{
      type=MovieScene3DTransformSection
      channels [
       Location.X {type=MovieSceneScriptingDoubleChannel, default=0}
       Location.Y {type=MovieSceneScriptingDoubleChannel, default=3.5978}
       Location.Z {type=MovieSceneScriptingDoubleChannel, default=22.3654}
       Rotation.X {type=MovieSceneScriptingDoubleChannel, default=0, keys [[282,0],[434,-16.3312]]}
       Rotation.Y {type=MovieSceneScriptingDoubleChannel, default=0, const [0,2]}
       Rotation.Z {type=MovieSceneScriptingDoubleChannel, default=0, const [0,2]}
       Scale.X {type=MovieSceneScriptingDoubleChannel, default=0.6541}
       Scale.Y {type=MovieSceneScriptingDoubleChannel, default=0.6541}
       Scale.Z {type=MovieSceneScriptingDoubleChannel, default=0.6541}
      ]
     }]
   }]
 }
 P1_Light_Head {
  id=#29
  possessed_class=Actor
  tracks [{
    type=MovieScene3DTransformTrack
    sections [{
      type=MovieScene3DTransformSection
      channels [
       Location.X {type=MovieSceneScriptingDoubleChannel, default=-0.2}
       Location.Y {type=MovieSceneScriptingDoubleChannel, default=30.5405}
       Location.Z {type=MovieSceneScriptingDoubleChannel, default=14.6283}
       Rotation.X {type=MovieSceneScriptingDoubleChannel, default=0, keys [[74,0],[240,-9.1155]]}
       Rotation.Y {type=MovieSceneScriptingDoubleChannel, default=0, const [0,2]}
       Rotation.Z {type=MovieSceneScriptingDoubleChannel, default=0, const [0,2]}
       Scale.X {type=MovieSceneScriptingDoubleChannel, default=1}
       Scale.Y {type=MovieSceneScriptingDoubleChannel, default=1}
       Scale.Z {type=MovieSceneScriptingDoubleChannel, default=1}
      ]
     }]
   }]
 }
 Panel1_Haylee_Root {
  id=#30
  possessed_class=Actor
  tracks [{
    type=MovieScene3DTransformTrack
    sections [{
      type=MovieScene3DTransformSection
      channels [
       Location.X {type=MovieSceneScriptingDoubleChannel, default=-78.2562, keys [[315,-69.7143],[330,-78.2562],[434,-80.8641]]}
       Location.Y {type=MovieSceneScriptingDoubleChannel, default=-8.7901, keys [[315,-83.009],[330,-8.7901],[434,-4.1798]]}
       Location.Z {type=MovieSceneScriptingDoubleChannel, default=0.249, keys [[315,57.7668],[330,0.249],[434,-2.1003]]}
       Rotation.X {type=MovieSceneScriptingDoubleChannel, default=14.5408, keys [[315,17.1465],[330,14.5408],[434,13.7389]]}
       Rotation.Y {type=MovieSceneScriptingDoubleChannel, default=0, const [0,3]}
       Rotation.Z {type=MovieSceneScriptingDoubleChannel, default=0, const [0,3]}
       Scale.X {type=MovieSceneScriptingDoubleChannel, default=1, keys [[315,0.6206],[330,1],[434,1]]}
       Scale.Y {type=MovieSceneScriptingDoubleChannel, default=1, keys [[315,0.6206],[330,1],[434,1]]}
       Scale.Z {type=MovieSceneScriptingDoubleChannel, default=1, keys [[315,0.6206],[330,1],[434,1]]}
      ]
     }]
   }]
 }
 P1_Ziggy_R_Hand {
  id=#31
  possessed_class=Actor
  tracks [{
    type=MovieScene3DTransformTrack
    sections [{
      type=MovieScene3DTransformSection
      channels [
       Location.X {type=MovieSceneScriptingDoubleChannel, default=0}
       Location.Y {type=MovieSceneScriptingDoubleChannel, default=16.077}
       Location.Z {type=MovieSceneScriptingDoubleChannel, default=-14.3676}
       Rotation.X {type=MovieSceneScriptingDoubleChannel, default=0, keys [[282,0],[434,17.0647]]}
       Rotation.Y {type=MovieSceneScriptingDoubleChannel, default=0, const [0,2]}
       Rotation.Z {type=MovieSceneScriptingDoubleChannel, default=0, const [0,2]}
       Scale.X {type=MovieSceneScriptingDoubleChannel, default=1}
       Scale.Y {type=MovieSceneScriptingDoubleChannel, default=1}
       Scale.Z {type=MovieSceneScriptingDoubleChannel, default=1}
      ]
     }]
   }]
 }
 P1_Synth_R_Hand {
  id=#32
  possessed_class=Actor
  tracks [{
    type=MovieScene3DTransformTrack
    sections [{
      type=MovieScene3DTransformSection
      channels [
       Location.X {type=MovieSceneScriptingDoubleChannel, default=0}
       Location.Y {type=MovieSceneScriptingDoubleChannel, default=-18.8787}
       Location.Z {type=MovieSceneScriptingDoubleChannel, default=-24.457}
       Rotation.X {type=MovieSceneScriptingDoubleChannel, default=0, keys [[235,0],[434,20.8394]]}
       Rotation.Y {type=MovieSceneScriptingDoubleChannel, default=0, const [0,1]}
       Rotation.Z {type=MovieSceneScriptingDoubleChannel, default=0, const [0,1]}
       Scale.X {type=MovieSceneScriptingDoubleChannel, default=0.6005}
       Scale.Y {type=MovieSceneScriptingDoubleChannel, default=0.6005}
       Scale.Z {type=MovieSceneScriptingDoubleChannel, default=0.6005}
      ]
     }]
   }]
 }
 P1_Synth_L_Hand {
  id=#33
  possessed_class=Actor
  tracks [{
    type=MovieScene3DTransformTrack
    sections [{
      type=MovieScene3DTransformSection
      channels [
       Location.X {type=MovieSceneScriptingDoubleChannel, default=0}
       Location.Y {type=MovieSceneScriptingDoubleChannel, default=-31.0865}
       Location.Z {type=MovieSceneScriptingDoubleChannel, default=7.0186}
       Rotation.X {type=MovieSceneScriptingDoubleChannel, default=0, keys [[235,0],[434,15.5795]]}
       Rotation.Y {type=MovieSceneScriptingDoubleChannel, default=0, const [0,1]}
       Rotation.Z {type=MovieSceneScriptingDoubleChannel, default=0, const [0,1]}
       Scale.X {type=MovieSceneScriptingDoubleChannel, default=0.6005}
       Scale.Y {type=MovieSceneScriptingDoubleChannel, default=0.6005}
       Scale.Z {type=MovieSceneScriptingDoubleChannel, default=0.6005}
      ]
     }]
   }]
 }
 Speed_Panel_1 {
  id=#34
  possessed_class=FortStaticMeshActor
  tracks [{
    type=MovieScene3DTransformTrack
    sections [{
      type=MovieScene3DTransformSection
      channels [
       Location.X {type=MovieSceneScriptingDoubleChannel, default=48.4861, const [48.4861,2]}
       Location.Y {type=MovieSceneScriptingDoubleChannel, default=-56.3826, keys [[10,-317.2269],[25,-69.5838]]}
       Location.Z {type=MovieSceneScriptingDoubleChannel, default=60.7329, keys [[10,-71.9103],[25,82.2374]]}
       Rotation.X {type=MovieSceneScriptingDoubleChannel, default=-90}
       Rotation.Y {type=MovieSceneScriptingDoubleChannel, default=58.4549}
       Rotation.Z {type=MovieSceneScriptingDoubleChannel, default=-90}
       Scale.X {type=MovieSceneScriptingDoubleChannel, default=1.5}
       Scale.Y {type=MovieSceneScriptingDoubleChannel, default=4}
       Scale.Z {type=MovieSceneScriptingDoubleChannel, default=1}
      ]
     }]
   }]
 }
 P1_Haylee_L_Leg {
  id=#35
  possessed_class=Actor
  tracks [{
    type=MovieScene3DTransformTrack
    sections [{
      type=MovieScene3DTransformSection
      channels [
       Location.X {type=MovieSceneScriptingDoubleChannel, default=0}
       Location.Y {type=MovieSceneScriptingDoubleChannel, default=4.1226}
       Location.Z {type=MovieSceneScriptingDoubleChannel, default=-6.3189}
       Rotation.X {type=MovieSceneScriptingDoubleChannel, default=0, keys [[315,0],[434,11.2909]]}
       Rotation.Y {type=MovieSceneScriptingDoubleChannel, default=0, const [0,2]}
       Rotation.Z {type=MovieSceneScriptingDoubleChannel, default=0, const [0,2]}
       Scale.X {type=MovieSceneScriptingDoubleChannel, default=0.4588}
       Scale.Y {type=MovieSceneScriptingDoubleChannel, default=0.4588}
       Scale.Z {type=MovieSceneScriptingDoubleChannel, default=0.4588}
      ]
     }]
   }]
 }
 P1_Ziggy_Spit {
  id=#36
  possessed_class=Actor
  tracks [{
    type=MovieScene3DTransformTrack
    sections [{
      type=MovieScene3DTransformSection
      channels [
       Location.X {type=MovieSceneScriptingDoubleChannel, default=0, const [0,2]}
       Location.Y {type=MovieSceneScriptingDoubleChannel, default=-4.5658, keys [[282,-4.5658],[434,-4.1108]]}
       Location.Z {type=MovieSceneScriptingDoubleChannel, default=-17.8057, keys [[282,-17.8057],[434,-18.1335]]}
       Rotation.X {type=MovieSceneScriptingDoubleChannel, default=0, keys [[282,0],[434,5.8201]]}
       Rotation.Y {type=MovieSceneScriptingDoubleChannel, default=0, const [0,2]}
       Rotation.Z {type=MovieSceneScriptingDoubleChannel, default=0, const [0,2]}
       Scale.X {type=MovieSceneScriptingDoubleChannel, default=0.6541, keys [[282,0.6541],[434,0.92]]}
       Scale.Y {type=MovieSceneScriptingDoubleChannel, default=0.6541, keys [[282,0.6541],[434,0.92]]}
       Scale.Z {type=MovieSceneScriptingDoubleChannel, default=0.6541, keys [[282,0.6541],[434,0.92]]}
      ]
     }]
   }]
 }
]
display_rate {denominator=1, numerator=60}
tick_resolution {denominator=1, numerator=24000}
tracks [{type=MovieSceneCameraCutTrack, sections [{type=MovieSceneCameraCutSection, end_frame=619, start_frame=0}]}]
