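# Level Sequence for the second panel of a comic-style intro, revealing six characters: Killswitch, Myst, Haylee, Ziggy, Light and Synth.
# Staggered transform keys bring each character controller into place with a small settling bounce, as impact roots pop from zero scale and slowly grow.
# Niagara hit bursts fire around each landing, with camera shakes on the first two.
# A cine camera pulls back to widen the shot while foreground embers drift.
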LevelSequence /PistolCanary/Layout/ComicPanels_v2/Level_Sequence/IntroPanel2_02
bindings [
 NiagaraComponent0 {
  id=#1
  parent_id=#2
  possessed_class=NiagaraComponent
  tracks [{
    type=MovieSceneParticleTrack
    display_name=FX System
    sections [{
      type=MovieSceneParticleSection
      channels [None {
        type=MovieSceneScriptingParticleChannel
        keys [[0, "<ParticleKey.DEACTIVATE: 1>"], [22, "<ParticleKey.ACTIVATE: 0>"], [43, "<ParticleKey.DEACTIVATE: 1>"]]
       }]
     }]
   }]
 }
 P2_Ziggy_Impact_Root2 {
  id=#3
  possessed_class=Actor
  tracks [{
    type=MovieScene3DTransformTrack
    sections [{
      type=MovieScene3DTransformSection
      channels [
       Location.X {type=MovieSceneScriptingDoubleChannel, default=0}
       Location.Y {type=MovieSceneScriptingDoubleChannel, default=0}
       Location.Z {type=MovieSceneScriptingDoubleChannel, default=0}
       Rotation.X {type=MovieSceneScriptingDoubleChannel, default=0}
       Rotation.Y {type=MovieSceneScriptingDoubleChannel, default=0}
       Rotation.Z {type=MovieSceneScriptingDoubleChannel, default=0}
       Scale.X {type=MovieSceneScriptingDoubleChannel, default=1, keys [[116,0],[120,1],[297,1.1046]]}
       Scale.Y {type=MovieSceneScriptingDoubleChannel, default=1, keys [[116,0],[120,1],[297,1.1046]]}
       Scale.Z {type=MovieSceneScriptingDoubleChannel, default=1, keys [[116,0],[120,1],[297,1.1046]]}
      ]
     }]
   }]
 }
 CameraComponent {
  id=#4
  parent_id=#5
  possessed_class=CineCameraComponent
  tracks [
   {
    type=MovieSceneFloatTrack
    display_name=Current Focal Length
    sections [{type=MovieSceneFloatSection, channels [None {type=MovieSceneScriptingFloatChannel, default=11.8}]}]
   }
   {
    type=MovieSceneFloatTrack
    display_name="Manual Focus Distance (Focus Settings)"
    sections [{type=MovieSceneFloatSection, channels [None {type=MovieSceneScriptingFloatChannel, default=100000}]}]
   }
   {
    type=MovieSceneFloatTrack
    display_name=Current Aperture
    sections [{type=MovieSceneFloatSection, channels [None {type=MovieSceneScriptingFloatChannel, default=2.8}]}]
   }
  ]
 }
 P2_Synth_Controller {
  id=#6
  possessed_class=Actor
  tracks [{
    type=MovieScene3DTransformTrack
    sections [{
      type=MovieScene3DTransformSection
      channels [
       Location.X {type=MovieSceneScriptingDoubleChannel, default=-3, keys [[140,8.503],[150,-3]]}
       Location.Y {type=MovieSceneScriptingDoubleChannel, default=-36.4902, keys [[140,-71.5052],[150,-36.4812]]}
       Location.Z {type=MovieSceneScriptingDoubleChannel, default=-6.6912, keys [[140,30.3846],[150,-6.7446],[153,-6.2606],[156,-6.7446]]}
       Rotation.X {type=MovieSceneScriptingDoubleChannel, default=90, const [90,2]}
       Rotation.Y {type=MovieSceneScriptingDoubleChannel, default=0, keys [[140,30.5274],[150,0]]}
       Rotation.Z {type=MovieSceneScriptingDoubleChannel, default=90, const [90,2]}
       Scale.X {type=MovieSceneScriptingDoubleChannel, default=0.1667}
       Scale.Y {type=MovieSceneScriptingDoubleChannel, default=0.1667}
       Scale.Z {type=MovieSceneScriptingDoubleChannel, default=0.1667}
      ]
     }]
   }]
 }
 P2_Light_Impact_Root2 {
  id=#7
  possessed_class=Actor
  tracks [{
    type=MovieScene3DTransformTrack
    sections [{
      type=MovieScene3DTransformSection
      channels [
       Location.X {type=MovieSceneScriptingDoubleChannel, default=0}
       Location.Y {type=MovieSceneScriptingDoubleChannel, default=0}
       Location.Z {type=MovieSceneScriptingDoubleChannel, default=0}
       Rotation.X {type=MovieSceneScriptingDoubleChannel, default=0}
       Rotation.Y {type=MovieSceneScriptingDoubleChannel, default=0}
       Rotation.Z {type=MovieSceneScriptingDoubleChannel, default=0}
       Scale.X {type=MovieSceneScriptingDoubleChannel, default=1, keys [[139,0],[143,1],[320,1.1046]]}
       Scale.Y {type=MovieSceneScriptingDoubleChannel, default=1, keys [[139,0],[143,1],[320,1.1046]]}
       Scale.Z {type=MovieSceneScriptingDoubleChannel, default=1, keys [[139,0],[143,1],[320,1.1046]]}
      ]
     }]
   }]
 }
 P2_Synth_Impact_Root2 {
  id=#8
  possessed_class=Actor
  tracks [{
    type=MovieScene3DTransformTrack
    sections [{
      type=MovieScene3DTransformSection
      channels [
       Location.X {type=MovieSceneScriptingDoubleChannel, default=0}
       Location.Y {type=MovieSceneScriptingDoubleChannel, default=0}
       Location.Z {type=MovieSceneScriptingDoubleChannel, default=0}
       Rotation.X {type=MovieSceneScriptingDoubleChannel, default=0}
       Rotation.Y {type=MovieSceneScriptingDoubleChannel, default=0}
       Rotation.Z {type=MovieSceneScriptingDoubleChannel, default=0}
       Scale.X {type=MovieSceneScriptingDoubleChannel, default=1, keys [[149,0],[153,1],[330,1.1046]]}
       Scale.Y {type=MovieSceneScriptingDoubleChannel, default=1, keys [[149,0],[153,1],[330,1.1046]]}
       Scale.Z {type=MovieSceneScriptingDoubleChannel, default=1, keys [[149,0],[153,1],[330,1.1046]]}
      ]
     }]
   }]
 }
 P2_Light_Controller {
  id=#9
  possessed_class=Actor
  tracks [{
    type=MovieScene3DTransformTrack
    sections [{
      type=MovieScene3DTransformSection
      channels [
       Location.X {type=MovieSceneScriptingDoubleChannel, default=-3, const [-3,2]}
       Location.Y {type=MovieSceneScriptingDoubleChannel, default=38.2097, keys [[128,63.0566],[140,38.2218]]}
       Location.Z {type=MovieSceneScriptingDoubleChannel, default=-7.4847, keys [[128,20.8996],[140,-7.499],[143,-7.0054],[146,-7.499]]}
       Rotation.X {type=MovieSceneScriptingDoubleChannel, default=90, const [90,2]}
       Rotation.Y {type=MovieSceneScriptingDoubleChannel, default=0, keys [[128,-25.4968],[140,0]]}
       Rotation.Z {type=MovieSceneScriptingDoubleChannel, default=90, const [90,2]}
       Scale.X {type=MovieSceneScriptingDoubleChannel, default=0.1667}
       Scale.Y {type=MovieSceneScriptingDoubleChannel, default=0.1667}
       Scale.Z {type=MovieSceneScriptingDoubleChannel, default=0.1667}
      ]
     }]
   }]
 }
 NE_Hit_FX_5 {id=#10, possessed_class=NiagaraActor, tracks []}
 P2_Myst_Controller {
  id=#11
  possessed_class=Actor
  tracks [{
    type=MovieScene3DTransformTrack
    sections [{
      type=MovieScene3DTransformSection
      channels [
       Location.X {type=MovieSceneScriptingDoubleChannel, default=-3, const [-3,1]}
       Location.Y {type=MovieSceneScriptingDoubleChannel, default=5.4454, const [5.445,1]}
       Location.Z {type=MovieSceneScriptingDoubleChannel, default=-9.9102, keys [[48,27.8438],[56,-9.9154],[59,-9.5873],[62,-9.9154]]}
       Rotation.X {type=MovieSceneScriptingDoubleChannel, default=90}
       Rotation.Y {type=MovieSceneScriptingDoubleChannel, default=0}
       Rotation.Z {type=MovieSceneScriptingDoubleChannel, default=90}
       Scale.X {type=MovieSceneScriptingDoubleChannel, default=0.1667}
       Scale.Y {type=MovieSceneScriptingDoubleChannel, default=0.1667}
       Scale.Z {type=MovieSceneScriptingDoubleChannel, default=0.1667}
      ]
     }]
   }]
 }
 NE_Hit_FX_6 {id=#12, possessed_class=NiagaraActor, tracks []}
 NiagaraComponent0 {
  id=#13
  parent_id=#12
  possessed_class=NiagaraComponent
  tracks [{
    type=MovieSceneParticleTrack
    display_name=FX System
    sections [{
      type=MovieSceneParticleSection
      channels [None {
        type=MovieSceneScriptingParticleChannel
        keys [
         [0, "<ParticleKey.DEACTIVATE: 1>"]
         [129, "<ParticleKey.DEACTIVATE: 1>"]
         [151, "<ParticleKey.ACTIVATE: 0>"]
         [172, "<ParticleKey.DEACTIVATE: 1>"]
        ]
       }]
     }]
   }]
 }
 NE_Hit_FX_2 {id=#14, possessed_class=NiagaraActor, tracks []}
 P2_Synth_FX_Root2 {
  id=#15
  possessed_class=Actor
  tracks [{
    type=MovieScene3DTransformTrack
    sections [{
      type=MovieScene3DTransformSection
      channels [
       Location.X {type=MovieSceneScriptingDoubleChannel, default=0}
       Location.Y {type=MovieSceneScriptingDoubleChannel, default=-5.3143}
       Location.Z {type=MovieSceneScriptingDoubleChannel, default=0}
       Rotation.X {type=MovieSceneScriptingDoubleChannel, default=0}
       Rotation.Y {type=MovieSceneScriptingDoubleChannel, default=0}
       Rotation.Z {type=MovieSceneScriptingDoubleChannel, default=0}
       Scale.X {type=MovieSceneScriptingDoubleChannel, default=1, keys [[149,0],[153,1],[203,1.005]]}
       Scale.Y {type=MovieSceneScriptingDoubleChannel, default=1, keys [[149,0],[153,1],[203,1.005]]}
       Scale.Z {type=MovieSceneScriptingDoubleChannel, default=1, keys [[149,0],[153,1],[203,1.005]]}
      ]
     }]
   }]
 }
 Haylee_Impact_Root2 {
  id=#16
  possessed_class=Actor
  tracks [{
    type=MovieScene3DTransformTrack
    sections [{
      type=MovieScene3DTransformSection
      channels [
       Location.X {type=MovieSceneScriptingDoubleChannel, default=3.5271}
       Location.Y {type=MovieSceneScriptingDoubleChannel, default=0.7304}
       Location.Z {type=MovieSceneScriptingDoubleChannel, default=-1.2222}
       Rotation.X {type=MovieSceneScriptingDoubleChannel, default=0}
       Rotation.Y {type=MovieSceneScriptingDoubleChannel, default=-90}
       Rotation.Z {type=MovieSceneScriptingDoubleChannel, default=-90}
       Scale.X {type=MovieSceneScriptingDoubleChannel, default=6.3819, keys [[106,0],[110,6.3819],[287,7]]}
       Scale.Y {type=MovieSceneScriptingDoubleChannel, default=6.3819, keys [[106,0],[110,6.3819],[287,7]]}
       Scale.Z {type=MovieSceneScriptingDoubleChannel, default=6.3819, keys [[106,0],[110,6.3819],[287,7]]}
      ]
     }]
   }]
 }
 P2_Haylee_Controller {
  id=#17
  possessed_class=Actor
  tracks [{
    type=MovieScene3DTransformTrack
    sections [{
      type=MovieScene3DTransformSection
      channels [
       Location.X {type=MovieSceneScriptingDoubleChannel, default=-3, keys [[97,4.0203],[107,-3]]}
       Location.Y {type=MovieSceneScriptingDoubleChannel, default=-22.0173, keys [[97,-37.9952],[107,-21.9525],[200,-22.0173]]}
       Location.Z {type=MovieSceneScriptingDoubleChannel, default=39, keys [[97,39],[107,-9.1461],[110,-8.6894],[113,-9.1461],[200,-9.1966]]}
       Rotation.X {type=MovieSceneScriptingDoubleChannel, default=90, const [90,3]}
       Rotation.Y {type=MovieSceneScriptingDoubleChannel, default=0, keys [[97,30.3412],[107,0]]}
       Rotation.Z {type=MovieSceneScriptingDoubleChannel, default=90, const [90,3]}
       Scale.X {type=MovieSceneScriptingDoubleChannel, default=0.1667}
       Scale.Y {type=MovieSceneScriptingDoubleChannel, default=0.1667}
       Scale.Z {type=MovieSceneScriptingDoubleChannel, default=0.1667}
      ]
     }]
   }]
 }
 P2_Killswitch_Controller {
  id=#18
  possessed_class=Actor
  tracks [{
    type=MovieScene3DTransformTrack
    sections [{
      type=MovieScene3DTransformSection
      channels [
       Location.X {type=MovieSceneScriptingDoubleChannel, default=-3, const [-3,1]}
       Location.Y {type=MovieSceneScriptingDoubleChannel, default=-6.9532, const [-6.9138,1]}
       Location.Z {type=MovieSceneScriptingDoubleChannel, default=-10.218, keys [[15,28.0235],[23,-10.2167],[26,-9.9674],[29,-10.2167]]}
       Rotation.X {type=MovieSceneScriptingDoubleChannel, default=90}
       Rotation.Y {type=MovieSceneScriptingDoubleChannel, default=0}
       Rotation.Z {type=MovieSceneScriptingDoubleChannel, default=90}
       Scale.X {type=MovieSceneScriptingDoubleChannel, default=0.1667}
       Scale.Y {type=MovieSceneScriptingDoubleChannel, default=0.1667}
       Scale.Z {type=MovieSceneScriptingDoubleChannel, default=0.1667}
      ]
     }]
   }]
 }
 NE_Hit_FX_3 {id=#19, possessed_class=NiagaraActor, tracks []}
 NiagaraComponent0 {
  id=#20
  parent_id=#19
  possessed_class=NiagaraComponent
  tracks [{
    type=MovieSceneParticleTrack
    display_name=FX System
    sections [{
      type=MovieSceneParticleSection
      channels [None {
        type=MovieSceneScriptingParticleChannel
        keys [
         [0, "<ParticleKey.DEACTIVATE: 1>"]
         [85, "<ParticleKey.DEACTIVATE: 1>"]
         [107, "<ParticleKey.ACTIVATE: 0>"]
         [128, "<ParticleKey.DEACTIVATE: 1>"]
        ]
       }]
     }]
   }]
 }
 NE_Hit_FX_01 {id=#2, possessed_class=NiagaraActor, tracks []}
 ComicIntro_Root {id=#21, possessed_class=TargetPoint, tracks []}
 NiagaraComponent0 {
  id=#22
  parent_id=#23
  possessed_class=NiagaraComponent
  tracks [{
    type=MovieSceneParticleTrack
    display_name=FX System
    sections [{
      type=MovieSceneParticleSection
      channels [None {
        type=MovieSceneScriptingParticleChannel
        keys [
         [0, "<ParticleKey.DEACTIVATE: 1>"]
         [95, "<ParticleKey.DEACTIVATE: 1>"]
         [117, "<ParticleKey.ACTIVATE: 0>"]
         [138, "<ParticleKey.DEACTIVATE: 1>"]
        ]
       }]
     }]
   }]
 }
 NiagaraComponent0 {
  id=#24
  parent_id=#10
  possessed_class=NiagaraComponent
  tracks [{
    type=MovieSceneParticleTrack
    display_name=FX System
    sections [{
      type=MovieSceneParticleSection
      channels [None {
        type=MovieSceneScriptingParticleChannel
        keys [
         [0, "<ParticleKey.DEACTIVATE: 1>"]
         [118, "<ParticleKey.DEACTIVATE: 1>"]
         [140, "<ParticleKey.ACTIVATE: 0>"]
         [161, "<ParticleKey.DEACTIVATE: 1>"]
        ]
       }]
     }]
   }]
 }
 P2_Myst_Impact_Root2 {
  id=#25
  possessed_class=Actor
  tracks [{
    type=MovieScene3DTransformTrack
    sections [{
      type=MovieScene3DTransformSection
      channels [
       Location.X {type=MovieSceneScriptingDoubleChannel, default=0}
       Location.Y {type=MovieSceneScriptingDoubleChannel, default=0}
       Location.Z {type=MovieSceneScriptingDoubleChannel, default=0}
       Rotation.X {type=MovieSceneScriptingDoubleChannel, default=0}
       Rotation.Y {type=MovieSceneScriptingDoubleChannel, default=0}
       Rotation.Z {type=MovieSceneScriptingDoubleChannel, default=0}
       Scale.X {type=MovieSceneScriptingDoubleChannel, default=1, keys [[55,0],[59,1],[236,1.1046]]}
       Scale.Y {type=MovieSceneScriptingDoubleChannel, default=1, keys [[55,0],[59,1],[236,1.1046]]}
       Scale.Z {type=MovieSceneScriptingDoubleChannel, default=1, keys [[55,0],[59,1],[236,1.1046]]}
      ]
     }]
   }]
 }
 NiagaraComponent0 {
  id=#26
  parent_id=#14
  possessed_class=NiagaraComponent
  tracks [{
    type=MovieSceneParticleTrack
    display_name=FX System
    sections [{
      type=MovieSceneParticleSection
      channels [None {
        type=MovieSceneScriptingParticleChannel
        keys [
         [0, "<ParticleKey.DEACTIVATE: 1>"]
         [34, "<ParticleKey.DEACTIVATE: 1>"]
         [56, "<ParticleKey.ACTIVATE: 0>"]
         [77, "<ParticleKey.DEACTIVATE: 1>"]
        ]
       }]
     }]
   }]
 }
 Camera_02 {
  id=#5
  possessed_class=CineCameraActor
  tracks [
   {
    type=MovieScene3DTransformTrack
    sections [{
      type=MovieScene3DTransformSection
      channels [
       Location.X {type=MovieSceneScriptingDoubleChannel, default=-49.3509, keys [[0,-20.0655],[78,-22.6571],[95,-43.4294],[203,-49.1759]]}
       Location.Y {type=MovieSceneScriptingDoubleChannel, default=0, const [0,1]}
       Location.Z {type=MovieSceneScriptingDoubleChannel, default=-1, keys [[78,-1.7007],[95,-1],[203,0]]}
       Rotation.X {type=MovieSceneScriptingDoubleChannel, default=0}
       Rotation.Y {type=MovieSceneScriptingDoubleChannel, default=1}
       Rotation.Z {type=MovieSceneScriptingDoubleChannel, default=-0.6}
       Scale.X {type=MovieSceneScriptingDoubleChannel, default=6}
       Scale.Y {type=MovieSceneScriptingDoubleChannel, default=6}
       Scale.Z {type=MovieSceneScriptingDoubleChannel, default=6}
      ]
     }]
   }
   {
    type=MovieSceneCameraShakeTrack
    sections [{type=MovieSceneCameraShakeSection, end_frame=323, start_frame=23}, {type=MovieSceneCameraShakeSection, end_frame=356, start_frame=56}]
   }
  ]
 }
 Panel_02_FG_Embers {
  id=#27
  possessed_class=FortStaticMeshActor
  tracks [{
    type=MovieScene3DTransformTrack
    sections [{
      type=MovieScene3DTransformSection
      channels [
       Location.X {type=MovieSceneScriptingDoubleChannel, default=-19.4495, keys [[71,-21.0965],[203,-17.9499]]}
       Location.Y {type=MovieSceneScriptingDoubleChannel, default=0, const [0,1]}
       Location.Z {type=MovieSceneScriptingDoubleChannel, default=-7.9943, const [-7.9943,1]}
       Rotation.X {type=MovieSceneScriptingDoubleChannel, default=90}
       Rotation.Y {type=MovieSceneScriptingDoubleChannel, default=0}
       Rotation.Z {type=MovieSceneScriptingDoubleChannel, default=90}
       Scale.X {type=MovieSceneScriptingDoubleChannel, default=0.655}
       Scale.Y {type=MovieSceneScriptingDoubleChannel, default=0.2177}
       Scale.Z {type=MovieSceneScriptingDoubleChannel, default=0.1706}
      ]
     }]
   }]
 }
 P2_Ziggy_Controller {
  id=#28
  possessed_class=Actor
  tracks [{
    type=MovieScene3DTransformTrack
    sections [{
      type=MovieScene3DTransformSection
      channels [
       Location.X {type=MovieSceneScriptingDoubleChannel, default=-3, keys [[109,9.6575],[117,-3]]}
       Location.Y {type=MovieSceneScriptingDoubleChannel, default=21.2167, keys [[109,13.7134],[117,21.2141]]}
       Location.Z {type=MovieSceneScriptingDoubleChannel, default=-7.499, keys [[109,41.6632],[117,-7.4867],[120,-7.0596],[123,-7.4867]]}
       Rotation.X {type=MovieSceneScriptingDoubleChannel, default=90, const [90,2]}
       Rotation.Y {type=MovieSceneScriptingDoubleChannel, default=0, keys [[109,21.4746],[117,0]]}
       Rotation.Z {type=MovieSceneScriptingDoubleChannel, default=90, const [90,2]}
       Scale.X {type=MovieSceneScriptingDoubleChannel, default=0.1667}
       Scale.Y {type=MovieSceneScriptingDoubleChannel, default=0.1667}
       Scale.Z {type=MovieSceneScriptingDoubleChannel, default=0.1667}
      ]
     }]
   }]
 }
 Killswitch_Impact_Root2 {
  id=#29
  possessed_class=Actor
  tracks [{
    type=MovieScene3DTransformTrack
    sections [{
      type=MovieScene3DTransformSection
      channels [
       Location.X {type=MovieSceneScriptingDoubleChannel, default=0}
       Location.Y {type=MovieSceneScriptingDoubleChannel, default=0}
       Location.Z {type=MovieSceneScriptingDoubleChannel, default=0}
       Rotation.X {type=MovieSceneScriptingDoubleChannel, default=90}
       Rotation.Y {type=MovieSceneScriptingDoubleChannel, default=0}
       Rotation.Z {type=MovieSceneScriptingDoubleChannel, default=-180}
       Scale.X {type=MovieSceneScriptingDoubleChannel, default=1.2, keys [[22,0],[26,1],[203,1.1046]]}
       Scale.Y {type=MovieSceneScriptingDoubleChannel, default=1.2, keys [[22,0],[26,1],[203,1.1046]]}
       Scale.Z {type=MovieSceneScriptingDoubleChannel, default=1.2, keys [[22,0],[26,1],[203,1.1046]]}
      ]
     }]
   }]
 }
 NE_Hit_FX_4 {id=#23, possessed_class=NiagaraActor, tracks []}
]
display_rate {denominator=1, numerator=60}
tick_resolution {denominator=1, numerator=24000}
tracks [{type=MovieSceneCameraCutTrack, sections [{type=MovieSceneCameraCutSection, end_frame=300, start_frame=0}]}]
