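# Level Sequence for a shot of comic intro panel 3, animating a cine camera and two "Myst" hand actors already in the level.
# Camera_03 moves quickly toward zero X and Z, then drifts slowly, with fixed focal length, focus distance and aperture.
# The right and left hand roots rotate in opposite directions on Z with the same fast-then-slow timing.
# One camera cut section covers the shot, and the ComicIntro_Root target point is bound without tracks.
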LevelSequence /PistolCanary/Layout/ComicPanels_v2/Level_Sequence/IntroPanel3_02
bindings [
 Camera_03 {
  id=#1
  possessed_class=CineCameraActor
  tracks [{
    type=MovieScene3DTransformTrack
    sections [{
      type=MovieScene3DTransformSection
      channels [
       Location.X {type=MovieSceneScriptingDoubleChannel, default=0, keys [[0,46.6417],[30,10.756],[364,0]]}
       Location.Y {type=MovieSceneScriptingDoubleChannel, default=0, const [0,3]}
       Location.Z {type=MovieSceneScriptingDoubleChannel, default=0, keys [[0,23.4555],[30,4.6121],[364,0]]}
       Rotation.X {type=MovieSceneScriptingDoubleChannel, default=0}
       Rotation.Y {type=MovieSceneScriptingDoubleChannel, default=0}
       Rotation.Z {type=MovieSceneScriptingDoubleChannel, default=0}
       Scale.X {type=MovieSceneScriptingDoubleChannel, default=1}
       Scale.Y {type=MovieSceneScriptingDoubleChannel, default=1}
       Scale.Z {type=MovieSceneScriptingDoubleChannel, default=1}
      ]
     }]
   }]
 }
 P3_Myst_R_Hand_Root2 {
  id=#2
  possessed_class=Actor
  tracks [{
    type=MovieScene3DTransformTrack
    sections [{
      type=MovieScene3DTransformSection
      channels [
       Location.X {type=MovieSceneScriptingDoubleChannel, default=-38.7756}
       Location.Y {type=MovieSceneScriptingDoubleChannel, default=-152.768}
       Location.Z {type=MovieSceneScriptingDoubleChannel, default=0}
       Rotation.X {type=MovieSceneScriptingDoubleChannel, default=0}
       Rotation.Y {type=MovieSceneScriptingDoubleChannel, default=-90}
       Rotation.Z {type=MovieSceneScriptingDoubleChannel, default=-85.9751, keys [[0,-85.7848],[45,-97.4562],[364,-109.2901]]}
       Scale.X {type=MovieSceneScriptingDoubleChannel, default=2}
       Scale.Y {type=MovieSceneScriptingDoubleChannel, default=2}
       Scale.Z {type=MovieSceneScriptingDoubleChannel, default=2}
      ]
     }]
   }]
 }
 P3_Myst_L_Hand_Root2 {
  id=#3
  possessed_class=Actor
  tracks [{
    type=MovieScene3DTransformTrack
    sections [{
      type=MovieScene3DTransformSection
      channels [
       Location.X {type=MovieSceneScriptingDoubleChannel, default=36.6386}
       Location.Y {type=MovieSceneScriptingDoubleChannel, default=-89.7479}
       Location.Z {type=MovieSceneScriptingDoubleChannel, default=0}
       Rotation.X {type=MovieSceneScriptingDoubleChannel, default=0, const [0,2]}
       Rotation.Y {type=MovieSceneScriptingDoubleChannel, default=-90, const [-90,2]}
       Rotation.Z {type=MovieSceneScriptingDoubleChannel, default=-90.6191, keys [[0,-90.3967],[30,-77.8726],[364,-54.7246]]}
       Scale.X {type=MovieSceneScriptingDoubleChannel, default=2}
       Scale.Y {type=MovieSceneScriptingDoubleChannel, default=2}
       Scale.Z {type=MovieSceneScriptingDoubleChannel, default=2}
      ]
     }]
   }]
 }
 ComicIntro_Root {id=#4, possessed_class=TargetPoint, tracks []}
 CameraComponent {
  id=#5
  parent_id=#1
  possessed_class=CineCameraComponent
  tracks [
   {
    type=MovieSceneFloatTrack
    display_name=Current Focal Length
    sections [{type=MovieSceneFloatSection, channels [None {type=MovieSceneScriptingFloatChannel, default=11.8}]}]
   }
   {
    type=MovieSceneFloatTrack
    display_name="Manual Focus Distance (Focus Settings)"
    sections [{type=MovieSceneFloatSection, channels [None {type=MovieSceneScriptingFloatChannel, default=100000}]}]
   }
   {
    type=MovieSceneFloatTrack
    display_name=Current Aperture
    sections [{type=MovieSceneFloatSection, channels [None {type=MovieSceneScriptingFloatChannel, default=2.8}]}]
   }
  ]
 }
]
display_rate {denominator=1, numerator=60}
tick_resolution {denominator=1, numerator=24000}
tracks [{type=MovieSceneCameraCutTrack, sections [{type=MovieSceneCameraCutSection, end_frame=376, start_frame=0}]}]
